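# M_LavaParticle is an unlit material for particle sprites, named for lava particles in an environment mechanics prototype.
# An orange-red color parameter, named "blueon", feeds its RGB into the Emissive input of the shared MF_FX_EmissiveAdjustments function.
# That function call is the only top-level node, so the material adds no textures or other logic of its own.
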Material /Game/Missions/Prototypes/EnvironmentMechanics/M_LavaParticle
graph {
 properties {bUsedWithParticleSprites=true, pixelDepthOffsetMode=PDOM_Legacy, shadingModel=MSM_Unlit, usedShadingModels=Unlit}
 nodes [MaterialFunctionCall(Emissive=VectorParameter() {parameterName=blueon, defaultValue=<rgba 1 0.1067 0 0>}.RGB) {materialFunction=&MF_FX_EmissiveAdjustments}]
}
