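# M_Shrine_Blue is a two-sided Default Lit prototype material, named for a blue shrine among the mission environment mechanics.
# Its graph holds a single vector parameter, blueon, which defaults to pure blue with zero alpha.
# No connection from that parameter to a material output is shown, so the graph defines no visible shading logic of its own.
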Material /Game/Missions/Prototypes/EnvironmentMechanics/M_Shrine_Blue
graph {
 properties {pixelDepthOffsetMode=PDOM_Legacy, twoSided=true, usedShadingModels=Default Lit}
 nodes [VectorParameter() {parameterName=blueon, defaultValue=<rgba 0 0 1 0>}]
}
