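# Prototype two-sided Default Lit material for a green shrine, judging by its name and environment-mechanics folder.
# Its graph holds a single vector parameter, greenon, which defaults to pure green with zero alpha.
# No link from that parameter to a material output appears in the graph, so it has no visible effect as recorded.
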Material /Game/Missions/Prototypes/EnvironmentMechanics/M_Shrine_Green
graph {
 properties {pixelDepthOffsetMode=PDOM_Legacy, twoSided=true, usedShadingModels=Default Lit}
 nodes [VectorParameter() {parameterName=greenon, defaultValue=<rgba 0 1 0 0>}]
}
