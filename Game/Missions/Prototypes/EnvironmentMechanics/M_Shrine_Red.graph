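# M_Shrine_Red is a two-sided Default Lit material in the EnvironmentMechanics mission prototypes, apparently the red variant for a shrine.
# Its only node is a vector parameter, RedOn, defaulting to opaque red.
# The graph shows no link from that parameter to any material output, and no textures or other logic.
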Material /Game/Missions/Prototypes/EnvironmentMechanics/M_Shrine_Red
graph {
 properties {pixelDepthOffsetMode=PDOM_Legacy, twoSided=true, usedShadingModels=Default Lit}
 nodes [VectorParameter() {parameterName=RedOn, defaultValue=<rgba 1 0 0 1>}]
}
